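# Translucent, two-sided lit material for flying-insect mesh particles (Cascade or Niagara) that animates flapping wings.
# A firefly texture, slightly brightened, is tinted by particle color, with alpha scaled by particle alpha and motion-blur fade.
# Flapping is a sine of time that pushes vertices along their normals, masked by red vertex color.
# Dynamic parameters give each particle its flap time offset, speed and amplitude scale.
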Material /Game/Effects/Fort_Effects/Materials/Fly/M_FlyWingFlap
graph {
 properties {
  bEnableResponsiveAA=true
  bUsedWithMeshParticles=true
  bUsedWithNiagaraMeshParticles=true
  blendMode=BLEND_Translucent
  pixelDepthOffsetMode=PDOM_Legacy
  translucencyLightingMode=TLM_VolumetricPerVertexNonDirectional
  twoSided=true
  usedShadingModels=Default Lit
 }
 nodes [
  Constant() {r=0.5}
  n1 = TextureSample() {texture=/Game/Effects/Fort_Effects/Textures/Fireflies/T_Firefly_Mesh_01}
  n2 = ParticleColor()
  MaterialFunctionCall(DiffuseColor=Multiply(A=Add(A=@n1.RGB) {constB=0.2}, B=@n2.RGB)) {materialFunction=&ConvertFromDiffSpec}
  Multiply(A=Multiply(A=@n1.A, B=@n2.A), B=ParticleMotionBlurFade())
  n3 = VertexColor()
  n4 = DynamicParameter() {defaultValue=<rgba 0 0 0 1>, paramNames [FlapOffset,ZFlapScalar,FlapSpeed,Param4]}
  n5 = Add(A=Time(), B=@n4.FlapOffset)
  n6 = Multiply(A=Sine(None=Multiply(A=@n5, B=@n4.FlapSpeed)), B=Multiply(A=Constant() {r=4}, B=@n4.ZFlapScalar))
  Multiply(A=Multiply(A=VertexNormalWS(), B=Multiply(A=@n3.R, B=@n6)))
  VectorParameter() {parameterName=SubSurfColor, defaultValue=<rgba 0.69 0.5743 0.414 1>}
 ]
}
