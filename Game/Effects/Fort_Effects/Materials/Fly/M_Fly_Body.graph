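# Lit, two-sided material for the body of a fly effect, built for Cascade and Niagara mesh particles.
# The firefly mesh texture is doubled, lifted slightly, clamped, then tinted by a fixed muted purple and the particle color.
# The result feeds the diffuse input of ConvertFromDiffSpec.
# Responsive AA is enabled, which suits small moving particles.
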Material /Game/Effects/Fort_Effects/Materials/Fly/M_Fly_Body
graph {
 properties {
  bEnableResponsiveAA=true
  bUsedWithMeshParticles=true
  bUsedWithNiagaraMeshParticles=true
  pixelDepthOffsetMode=PDOM_Legacy
  twoSided=true
  usedShadingModels=Default Lit
 }
 nodes [
  Constant() {r=1}
  Constant() {r=0.4}
  n1 = TextureSample() {texture=/Game/Effects/Fort_Effects/Textures/Fireflies/T_Firefly_Mesh_01}
  n2 = Clamp(None=Add(A=Multiply(A=@n1.RGB) {constB=2}) {constB=0.1})
  n3 = Constant3Vector() {constant=<rgba 0.3434 0.2626 0.537 0>}
  MaterialFunctionCall(DiffuseColor=Multiply(A=Multiply(A=@n2, B=@n3), B=ParticleColor().RGB)) {materialFunction=&ConvertFromDiffSpec}
 ]
}
